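# Physics asset for the short pencil skirt lower-body cosmetic mesh on the medium female player body.
# Capsules on the pelvis and thighs represent the body, alongside two-segment skirt bone chains at the back center, back left and right, and both sides.
# Each chain hangs from a small base capsule joined to the pelvis with free angular motion; the front has only a base.
# Tight swing and twist limits shape the chains, with back panels swinging wider than sides.
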PhysicsAsset /BRCosmetics/Characters/Player/Female/Medium/Bodies/F_MED_Pencil_B/SkirtShort/Meshes/Pencil_SkirtShort_Blazer_MedB_Lbdy_Physics
bodies {
 dyn_skirt_bk_a [{shapeType=Capsule, length=3.125, radius=4.375, shapeName=dyn_skirt_bk_a_capsule, center=<7.063 -2.51 0>, rotation=<rot 0 -92.4227 -91.2533>}]
 dyn_skirt_bk_a_base [{shapeType=Capsule, length=0.505, radius=0.505, shapeName=dyn_skirt_bk_a_base_capsule, rotation=<rot 0 -90 0>}]
 dyn_skirt_bk_a_base_l [{shapeType=Capsule, length=0.505, radius=0.505, shapeName=dyn_skirt_bk_a_base_l_capsule, rotation=<rot 0 -90 0>}]
 dyn_skirt_bk_a_base_r [{shapeType=Capsule, length=0.505, radius=0.505, shapeName=dyn_skirt_bk_a_base_r_capsule, rotation=<rot 0 -90 0>}]
 dyn_skirt_bk_a_l [{
   shapeType=Capsule
   length=4.75
   radius=5.375
   shapeName=dyn_skirt_bk_a_l_capsule
   center=<-6.9613 4 -2>
   rotation=<rot 0.1683 -97.7978 -89.0458>
  }]
 dyn_skirt_bk_a_r [{shapeType=Capsule, length=1.75, radius=5.75, shapeName=dyn_skirt_bk_a_r_capsule, center=<7 -4 2>, rotation=<rot 0 90 -90>}]
 dyn_skirt_bk_b [{
   shapeType=Capsule
   length=2.5
   radius=5
   shapeName=dyn_skirt_bk_b_capsule
   center=<7.2269 1.0379 -0.0328>
   rotation=<rot 0.7018 -94.8595 87.2689>
  }]
 dyn_skirt_bk_b_l [{shapeType=Capsule, length=2.25, radius=5.25, shapeName=dyn_skirt_bk_b_l_capsule, center=<-7 0 0>, rotation=<rot 14.2138 -89.9999 -90.0676>}]
 dyn_skirt_bk_b_r [{
   shapeType=Capsule
   length=5.6781
   radius=8.3225
   shapeName=dyn_skirt_bk_b_r_capsule
   center=<10.0201 -3 0>
   rotation=<rot 14.3215 -86.127 -89.1214>
  }]
 dyn_skirt_fr_a_base [{shapeType=Capsule, length=0.505, radius=0.505, shapeName=dyn_skirt_fr_a_base_capsule, rotation=<rot 0 -90 0>}]
 dyn_skirt_side_a_base_l [{shapeType=Capsule, length=0.505, radius=0.505, shapeName=dyn_skirt_side_a_base_l_capsule, rotation=<rot 0 -90 0>}]
 dyn_skirt_side_a_base_r [{shapeType=Capsule, length=0.505, radius=0.505, shapeName=dyn_skirt_side_a_base_r_capsule, rotation=<rot 0 -90 0>}]
 dyn_skirt_side_a_l [{
   shapeType=Capsule
   length=7.375
   radius=6.375
   shapeName=dyn_skirt_side_a_l_capsule
   center=<7 -4 0>
   rotation=<rot 4.3826 -89.7659 -176.9721>
  }]
 dyn_skirt_side_a_r [{
   shapeType=Capsule
   length=5.625
   radius=8.125
   shapeName=dyn_skirt_side_a_l_capsule
   center=<7 5 0>
   rotation=<rot 4.3826 -89.7659 -176.9722>
  }]
 dyn_skirt_side_b_l [{
   shapeType=Capsule
   length=12
   radius=8.5
   shapeName=dyn_skirt_side_b_l_capsule
   center=<6.9838 -3.5 0>
   rotation=<rot 4.3731 -90.0169 -0.404>
  }]
 dyn_skirt_side_b_r [{shapeType=Capsule, length=11.75, radius=8.75, shapeName=dyn_skirt_side_b_l_capsule, center=<7 3.5 0>, rotation=<rot 4.3732 -90.0169 -0.4039>}]
 pelvis [{shapeType=Capsule, length=10.1069, radius=9.8036, center=<-8.3424 -1.3543 0>, rotation=<rot 0 -90 0>}]
 thigh_l [{shapeType=Capsule, length=27.75, radius=7.25, shapeName=thigh_l_capsule, center=<-16.9587 0.7457 0.6245>, rotation=<rot 0 -86.5304 -92.7649>}]
 thigh_r [{shapeType=Capsule, length=26.75, radius=8.25, shapeName=thigh_l_capsule, center=<16.9587 -0.7457 -0.6245>, rotation=<rot 0 -86.5304 87.2351>}]
}
constraints [
 {
  bone1Name=thigh_r
  bone2Name=pelvis
  swing1LimitDegrees=45
  swing1Motion=Free
  swing2LimitDegrees=45
  swing2Motion=Free
  twistLimitDegrees=45
  twistMotion=Free
 }
 {
  bone1Name=dyn_skirt_bk_a_base
  bone2Name=pelvis
  swing1LimitDegrees=45
  swing1Motion=Free
  swing2LimitDegrees=45
  swing2Motion=Free
  twistLimitDegrees=45
  twistMotion=Free
 }
 {
  bone1Name=dyn_skirt_bk_a_base_l
  bone2Name=pelvis
  swing1LimitDegrees=45
  swing1Motion=Free
  swing2LimitDegrees=45
  swing2Motion=Free
  twistLimitDegrees=45
  twistMotion=Free
 }
 {
  bone1Name=dyn_skirt_bk_a_base_r
  bone2Name=pelvis
  swing1LimitDegrees=45
  swing1Motion=Free
  swing2LimitDegrees=45
  swing2Motion=Free
  twistLimitDegrees=45
  twistMotion=Free
 }
 {
  bone1Name=dyn_skirt_fr_a_base
  bone2Name=pelvis
  swing1LimitDegrees=45
  swing1Motion=Free
  swing2LimitDegrees=45
  swing2Motion=Free
  twistLimitDegrees=45
  twistMotion=Free
 }
 {
  bone1Name=dyn_skirt_side_a_base_l
  bone2Name=pelvis
  swing1LimitDegrees=45
  swing1Motion=Free
  swing2LimitDegrees=45
  swing2Motion=Free
  twistLimitDegrees=45
  twistMotion=Free
 }
 {
  bone1Name=dyn_skirt_side_a_base_r
  bone2Name=pelvis
  swing1LimitDegrees=45
  swing1Motion=Free
  swing2LimitDegrees=45
  swing2Motion=Free
  twistLimitDegrees=45
  twistMotion=Free
 }
 {bone1Name=dyn_skirt_bk_a, bone2Name=dyn_skirt_bk_a_base, swing1LimitDegrees=20, swing2LimitDegrees=10, twistLimitDegrees=5}
 {bone1Name=dyn_skirt_bk_b, bone2Name=dyn_skirt_bk_a, swing1LimitDegrees=20, swing2LimitDegrees=10, twistLimitDegrees=5}
 {bone1Name=dyn_skirt_bk_a_l, bone2Name=dyn_skirt_bk_a_base_l, swing1LimitDegrees=20, swing2LimitDegrees=5, twistLimitDegrees=5}
 {bone1Name=dyn_skirt_bk_b_l, bone2Name=dyn_skirt_bk_a_l, swing1LimitDegrees=20, swing2LimitDegrees=5, twistLimitDegrees=5}
 {bone1Name=dyn_skirt_bk_a_r, bone2Name=dyn_skirt_bk_a_base_r, swing1LimitDegrees=20, swing2LimitDegrees=3, twistLimitDegrees=3}
 {bone1Name=dyn_skirt_bk_b_r, bone2Name=dyn_skirt_bk_a_r, swing1LimitDegrees=20, swing2LimitDegrees=3, twistLimitDegrees=3}
 {bone1Name=dyn_skirt_side_a_l, bone2Name=dyn_skirt_side_a_base_l, swing1LimitDegrees=5, swing2LimitDegrees=5, twistLimitDegrees=5}
 {bone1Name=dyn_skirt_side_b_l, bone2Name=dyn_skirt_side_a_l, swing1LimitDegrees=5, swing2LimitDegrees=5, twistLimitDegrees=5}
 {bone1Name=dyn_skirt_side_a_r, bone2Name=dyn_skirt_side_a_base_r, swing1LimitDegrees=5, swing2LimitDegrees=5, twistLimitDegrees=5}
 {bone1Name=dyn_skirt_side_b_r, bone2Name=dyn_skirt_side_a_r, swing1LimitDegrees=5, swing2LimitDegrees=5, twistLimitDegrees=5}
 {
  bone1Name=thigh_l
  bone2Name=pelvis
  swing1LimitDegrees=45
  swing1Motion=Free
  swing2LimitDegrees=45
  swing2Motion=Free
  twistLimitDegrees=45
  twistMotion=Free
 }
]
