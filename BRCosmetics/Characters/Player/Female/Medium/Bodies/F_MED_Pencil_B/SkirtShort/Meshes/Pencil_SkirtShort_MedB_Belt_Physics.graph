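# Physics asset for the belt on the Pencil short-skirt cosmetic for the medium female player body B, simulating its dangling bones.
# Capsule bodies on dyn_belt_MSTR, dyn_belt_01 and dyn_belt_02 form a two-constraint chain, with one-degree swing limit values and zero twist.
# A long left-thigh capsule sits outside the chain with no constraint, apparently as a collision shape for the belt.
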PhysicsAsset /BRCosmetics/Characters/Player/Female/Medium/Bodies/F_MED_Pencil_B/SkirtShort/Meshes/Pencil_SkirtShort_MedB_Belt_Physics
bodies {
 dyn_belt_01 [{shapeType=Capsule, length=2.1627, radius=2.4282, shapeName=dyn_belt_01_capsule, center=<2.2728 0 0>, rotation=<rot 90 0 0>}]
 dyn_belt_02 [{shapeType=Capsule, length=7.7227, radius=2.8803, shapeName=dyn_belt_02_capsule, center=<3.9137 0 0.3841>, rotation=<rot 90 0 0>}]
 dyn_belt_MSTR [{shapeType=Capsule, length=4.3914, radius=1.9222, shapeName=dyn_belt_02_capsule, rotation=<rot 0 -94.9325 0>}]
 thigh_l [{shapeType=Capsule, length=54.6076, radius=8.1536, shapeName=thigh_l_capsule, center=<-7.7357 -0.6835 2.8043>, rotation=<rot 90 0 0>}]
}
constraints [
 {bone1Name=dyn_belt_01, bone2Name=dyn_belt_MSTR, swing1LimitDegrees=1, swing2LimitDegrees=1, twistLimitDegrees=0}
 {bone1Name=dyn_belt_02, bone2Name=dyn_belt_01, swing1LimitDegrees=1, swing2LimitDegrees=1, twistLimitDegrees=0}
]
