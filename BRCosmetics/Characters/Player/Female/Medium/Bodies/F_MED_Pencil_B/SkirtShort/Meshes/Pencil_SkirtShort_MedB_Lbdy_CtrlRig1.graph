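# Control Rig for the short pencil skirt on the medium female player body, correcting skirt base bones as the thighs move.
# A spherical pose reader per thigh feeds a smoothed alpha that weights transform offsets on that side's skirt bones.
# The emote variable caps both alphas and defaults to zero, so correctives stay off until it is raised.
# Parent constraints then pull the center front bone toward each front side bone at half that side's alpha.
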ControlRigBlueprint /BRCosmetics/Characters/Player/Female/Medium/Bodies/F_MED_Pencil_B/SkirtShort/Meshes/Pencil_SkirtShort_MedB_Lbdy_CtrlRig1
graphs [
 RigVMFunctionLibrary {links [], nodes []}
 RigVMModel {nodes [
   RigUnit_BeginExecution = &RigUnit_BeginExecution()
   SequenceAggregate = &RigVMFunction_Sequence(ExecuteContext=@RigUnit_BeginExecution.ExecutePin)
   n1 = &RigUnit_SphericalPoseReader(
    ActiveRegionScaleFactors=(PositiveWidth=3.5,NegativeWidth=0.5,PositiveHeight=3,NegativeHeight=3)
    ActiveRegionSize=0.1
    Debug=(bDrawDebug=true,DebugScale=25,DebugSegments=20,DebugThickness=0.25)
    DriverAxis=(X=-1,Y=0,Z=0)
    DriverItem=(Type=Bone,Name="thigh_l")
    FalloffRegionScaleFactors=(PositiveWidth=1,NegativeWidth=1,PositiveHeight=1,NegativeHeight=1)
    FalloffSize=0.45
    OptionalParentItem=(Type=Bone)
    RotationOffset=(X=135,Y=-10,Z=10)
    ExecutePin=@SequenceAggregate.A
   )
   > n2 = &RigUnit_SphericalPoseReader(
    ActiveRegionScaleFactors=(PositiveWidth=3.5,NegativeWidth=0.5,PositiveHeight=3,NegativeHeight=3)
    ActiveRegionSize=0.1
    Debug=(bDrawDebug=true,DebugScale=25,DebugSegments=20,DebugThickness=0.25)
    DriverAxis=(X=1,Y=0,Z=0)
    DriverItem=(Type=Bone,Name="thigh_r")
    FalloffRegionScaleFactors=(PositiveWidth=1,NegativeWidth=1,PositiveHeight=1,NegativeHeight=1)
    FalloffSize=0.45
    FlipHeightScaling=false
    FlipWidthScaling=true
    OptionalParentItem=(Type=Bone)
    RotationOffset=(X=-40,Y=-15,Z=15)
   )
   n3 = &RigVMFunction_AlphaInterp(
    InterpSpeedDecreasing=10
    InterpSpeedIncreasing=10
    Scale=1
    bClampResult=true
    Value=@n1.OutputParam
    ClampMax=RigVMVariableNode() {variable=emote}.Value
   )
   > &RigUnit_ModifyTransforms(
    ItemToModify="((Item=(Type=Bone,Name=\"dyn_skirt_fr_a_base_l\"),Transform=(Rotation=(X=0,Y=0,Z=0,W=1),Translation=(X=0,Y=0,Z=0),Scale3D=(X=1,Y=1,Z=1))),(Item=(Type=Bone,Name=\"dyn_skirt_side_a_base_l\"),Transform=(Rotation=(X=0,Y=0,Z=0,W=1),Translation=(X=-1,Y=-2,Z=0),Scale ...(truncated)"
    Weight=@n3.Result
    ItemToModify.0.Transform.Rotation=&RigVMFunction_MathQuaternionFromEuler(Euler=(X=0,Y=10,Z=15)).Result
    ItemToModify.1.Transform.Rotation=&RigVMFunction_MathQuaternionFromEuler(Euler=(X=0,Y=0,Z=-15)).Result
    ItemToModify.2.Transform.Rotation=&RigVMFunction_MathQuaternionFromEuler(Euler=(X=0,Y=0,Z=-15)).Result
   )
   n4 = &RigVMFunction_AlphaInterp(
    InterpSpeedDecreasing=10
    InterpSpeedIncreasing=10
    Scale=1
    bClampResult=true
    Value=@n2.OutputParam
    ClampMax=RigVMVariableNode() {variable=emote}.Value
   )
   > &RigUnit_ModifyTransforms(
    ItemToModify="((Item=(Type=Bone,Name=\"dyn_skirt_fr_a_base_r\"),Transform=(Rotation=(X=0,Y=0,Z=0,W=1),Translation=(X=-2,Y=-2,Z=-3),Scale3D=(X=1,Y=1,Z=1))),(Item=(Type=Bone,Name=\"dyn_skirt_side_a_base_r\"),Transform=(Rotation=(X=0,Y=0,Z=0,W=1),Translation=(X=-2,Y=5,Z=0),Sca ...(truncated)"
    Weight=@n4.Result
    ItemToModify.1.Transform.Rotation=&RigVMFunction_MathQuaternionFromEuler(Euler=(X=0,Y=0,Z=15)).Result
    ItemToModify.2.Transform.Rotation=&RigVMFunction_MathQuaternionFromEuler(Euler=(X=0,Y=0,Z=-15)).Result
    ItemToModify.0.Transform.Rotation=&RigVMFunction_MathQuaternionFromEuler(Euler=(X=0,Y=10,Z=15)).Result
   )
   Multiply = &RigVMFunction_MathFloatMul(B=0.5, A=@n4.Result) {template="Multiply::Execute(in A,in B,out Result)"}
   &RigUnit_ParentConstraint(
    AdvancedSettings=(RotationOrderForFilter=XZY)
    Child=(Type=Bone,Name="dyn_skirt_fr_a_base")
    Filter=(TranslationFilter=(bX=false,bY=false,bZ=false),RotationFilter=(bX=True,bY=True,bZ=True),ScaleFilter=(bX=True,bY=True,bZ=True))
    Parents=((Item=(Type=Bone,Name="dyn_skirt_fr_a_base_r"),Weight=1))
    Weight=@Multiply.Result
    ExecutePin=@SequenceAggregate.B
   )
   > &RigUnit_ParentConstraint(
    AdvancedSettings=(RotationOrderForFilter=XZY)
    Child=(Type=Bone,Name="dyn_skirt_fr_a_base")
    Filter=(TranslationFilter=(bX=false,bY=false,bZ=false),RotationFilter=(bX=true,bY=true,bZ=True),ScaleFilter=(bX=True,bY=True,bZ=True))
    Parents=((Item=(Type=Bone,Name="dyn_skirt_fr_a_base_l"),Weight=1))
    Weight=&RigVMFunction_MathFloatMul(B=0.5, A=@n3.Result) {template="Multiply::Execute(in A,in B,out Result)"}.Result
   )
  ]}
]
hierarchy [
 root {loc=<0 0 0>, rot=<0 0 0 1>}
 pelvis {parent=root, loc=<0 -0.5301 98.0571>, rot=<0.0127 0.707 -0.0127 -0.707>}
 spine_01 {parent=pelvis, loc=<6.1208 -0.9328 0>, rot=<0 0 0.034 -0.9994>}
 spine_02 {parent=spine_01, loc=<6.5996 -0.2095 0>, rot=<0 0 0.0238 -0.9997>}
 spine_03 {parent=spine_02, loc=<6.4914 -0.1237 0>, rot=<0 0 0.0149 -0.9999>}
 spine_04 {parent=spine_03, loc=<6.9598 0.0794 0>, rot=<0 0 -0.0132 -0.9999>}
 thigh_l {parent=pelvis, loc=<-1.0855 -0.2109 -8.7053>, rot=<0.0009 0.0355 0.0155 -0.9993>}
 calf_l {parent=thigh_l, loc=<-42.5284 -0.0019 -0.7297>, rot=<0.0006 -0.0161 -0.0391 -0.9991>}
 foot_l {parent=calf_l, loc=<-46.7401 -2.5014 -0.592>, rot=<0.0427 -0.0238 0.0353 -0.9982>}
 ball_l {parent=foot_l, loc=<-3.5658 10.8066 -0.1544>, rot=<0.0001 -0.0001 -0.7063 -0.7079>}
 calf_twist_02_l {parent=calf_l, loc=<-19.1546 -1.7943 -0.2638>, rot=<0.0237 0.0095 -0.0035 -0.9997>}
 calf_twist_01_l {parent=calf_l, loc=<-35.8483 -1.9197 -0.5794>, rot=<0.0237 0.0095 -0.0035 -0.9997>}
 thigh_twist_01_l {parent=thigh_l, loc=<-27.952 0 0>, rot=<0.0013 0 0 -1>}
 thigh_r {parent=pelvis, loc=<-1.0855 -0.2109 8.7053>, rot=<0.0355 -0.0009 0.9993 0.0155>}
 calf_r {parent=thigh_r, loc=<42.5284 0.0019 0.7297>, rot=<0.0006 -0.0161 -0.0391 -0.9991>}
 foot_r {parent=calf_r, loc=<46.7401 2.5014 0.592>, rot=<0.0427 -0.0238 0.0353 -0.9982>}
 ball_r {parent=foot_r, loc=<3.5658 -10.8066 0.1544>, rot=<0.0001 -0.0001 -0.7063 -0.7079>}
 calf_twist_02_r {parent=calf_r, loc=<19.1546 1.7943 0.2638>, rot=<0.0237 0.0095 -0.0035 -0.9997>}
 calf_twist_01_r {parent=calf_r, loc=<35.8483 1.9197 0.5794>, rot=<0.0237 0.0095 -0.0035 -0.9997>}
 thigh_twist_01_r {parent=thigh_r, loc=<27.952 0 0>, rot=<0.0013 0 0 -1>}
 dyn_skirt_bk_a_base {parent=pelvis, loc=<2.2496 -11.085 0.0859>, rot=<0.094 -0.9956 0.0002 -0.0018>}
 dyn_skirt_bk_a {parent=dyn_skirt_bk_a_base, loc=<0 0 0>, rot=<0 0 0 1>}
 dyn_skirt_bk_b {parent=dyn_skirt_bk_a, loc=<12.7925 0 0>, rot=<0 0 0.0017 -1>}
 dyn_skirt_fr_a_base {parent=pelvis, loc=<1.3903 10.8025 -0.0799>, rot=<-0.088 -0.9961 0.0001 0.0018>}
 dyn_skirt_fr_a {parent=dyn_skirt_fr_a_base, loc=<0 0 0>, rot=<0 0 0 1>}
 dyn_skirt_fr_b {parent=dyn_skirt_fr_a, loc=<11.0638 0 0>, rot=<0 0 0 1>}
 dyn_skirt_side_a_base_r {parent=pelvis, loc=<2.7203 0.125 13.8078>, rot=<0.1235 0.6916 0.7007 -0.1243>}
 dyn_skirt_side_a_r {parent=dyn_skirt_side_a_base_r, loc=<0 0 0>, rot=<0 0 0 1>}
 dyn_skirt_side_b_r {parent=dyn_skirt_side_a_r, loc=<13.4807 -0.0277 0>, rot=<0 0 -0.0017 -1>}
 dyn_skirt_fr_a_base_r {parent=pelvis, loc=<1.665 8.6787 7.8782>, rot=<-0.0774 -0.9908 -0.0917 0.0631>}
 dyn_skirt_fr_a_r {parent=dyn_skirt_fr_a_base_r, loc=<0 0 0>, rot=<0 0 0 1>}
 dyn_skirt_fr_b_r {parent=dyn_skirt_fr_a_r, loc=<11.4921 0 0>, rot=<0 0 -0.0017 -1>}
 dyn_skirt_bk_a_base_r {parent=pelvis, loc=<2.864 -9.087 8.088>, rot=<0.1075 -0.9235 0.3616 0.069>}
 dyn_skirt_bk_a_r {parent=dyn_skirt_bk_a_base_r, loc=<0 0 0>, rot=<0 0 0 1>}
 dyn_skirt_bk_b_r {parent=dyn_skirt_bk_a_r, loc=<13.5723 0 0>, rot=<0 0 0.0017 -1>}
 dyn_skirt_bk_a_base_l {parent=pelvis, loc=<2.8637 -9.087 -8.0885>, rot=<0.9235 0.1075 0.069 -0.3616>}
 dyn_skirt_bk_a_l {parent=dyn_skirt_bk_a_base_l, loc=<0 0 0>, rot=<0 0 0 1>}
 dyn_skirt_bk_b_l {parent=dyn_skirt_bk_a_l, loc=<-13.5532 0 0>, rot=<0 0 0 1>}
 dyn_skirt_side_a_base_l {parent=pelvis, loc=<2.7457 0.3012 -13.8>, rot=<0.1268 -0.7011 -0.6912 -0.1211>}
 dyn_skirt_side_a_l {parent=dyn_skirt_side_a_base_l, loc=<0 0 0>, rot=<0 0 0 1>}
 dyn_skirt_side_b_l {parent=dyn_skirt_side_a_l, loc=<13.4394 0 0>, rot=<0 0 0.0017 -1>}
 dyn_skirt_fr_a_base_l {parent=pelvis, loc=<1.8202 8.6563 -7.8566>, rot=<0.9908 -0.0774 0.0631 0.0917>}
 dyn_skirt_fr_a_l {parent=dyn_skirt_fr_a_base_l, loc=<0 0 0>, rot=<0 0 0 1>}
 dyn_skirt_fr_b_l {parent=dyn_skirt_fr_a_l, loc=<-11.5464 0 0>, rot=<0 0 -0.0017 -1>}
 dyn_belt_MSTR {parent=pelvis, loc=<5.1441 10.7281 2.2318>, rot=<0.4493 -0.5755 -0.5298 -0.4315>}
 dyn_belt_01 {parent=dyn_belt_MSTR, loc=<1.7187 0 0>, rot=<-0.0018 -0.0321 -0.0367 -0.9988>}
 dyn_belt_02 {parent=dyn_belt_01, loc=<6.176 0.0275 0.1841>, rot=<-0.0175 -0.2434 -0.0113 -0.9697>}
 dyn_belt_left {parent=pelvis, loc=<6.5621 6.6404 -10.5623>, rot=<-0.0127 -0.707 0.0127 -0.707>}
 Offset {type=Curve, loc=<0 0 0>, rot=<0 0 0 1>}
]
variables [emote {type=double, default=0}]
