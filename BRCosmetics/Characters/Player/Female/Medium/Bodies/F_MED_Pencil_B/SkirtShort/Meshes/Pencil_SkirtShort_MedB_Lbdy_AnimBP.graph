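# Post-processes the short skirt and belt of the Pencil female medium body, following the attached parent's pose.
# Two branches chain pelvis/thigh constraints, a Control Rig and pelvis-space RigidBody physics, selected by vest-jacket or blazer cosmetic tags.
# Bone offsets then correct crouch-moving, jumping/falling, front-end, sliding and riding poses, before a final emote Control Rig.
# The EventGraph derives movement direction, turning, state and outfit values each update.
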AnimBlueprint /BRCosmetics/Characters/Player/Female/Medium/Bodies/F_MED_Pencil_B/SkirtShort/Meshes/Pencil_SkirtShort_MedB_Lbdy_AnimBP
target_skeleton=&/Pencil_SkirtShort_MedB_Lbdy_Skeleton
graphs [
 AnimGraph {nodes [
   n1 = Get bIsRiding() -> (bIsRiding:bool)
   n2 = "Transform (Modify) Bone - Bone: dyn_skirt_fr_a_base_r"!(ComponentPose, bAlphaBoolEnabled:bool=@n1, Translation="0.005,0,-10", Rotation="0,5,0", Scale="1,1,1") -> (Pose) {
    alphaInputType=Bool
    rotationMode=BMM_Additive
    rotationSpace=BCS_BoneSpace
    translationMode=BMM_Additive
    translationSpace=BCS_BoneSpace
    alphaBoolBlend {blendInTime=0.1, blendOption=Linear, blendOutTime=0.3}
    boneToModify {boneName=dyn_skirt_fr_a_base_r}
    rotation {yaw=5}
    translation {x=0.005, z=-10}
   }
   n3 = "Transform (Modify) Bone - Bone: dyn_skirt_side_a_base_r"!(ComponentPose=@n2, bAlphaBoolEnabled:bool=@n1, Translation="0,10,-5", Rotation="0,25,0", Scale="1,1,1") -> (Pose) {
    alphaInputType=Bool
    rotationMode=BMM_Additive
    rotationSpace=BCS_BoneSpace
    translationMode=BMM_Additive
    translationSpace=BCS_BoneSpace
    alphaBoolBlend {blendInTime=0.1, blendOption=Linear, blendOutTime=0.3}
    boneToModify {boneName=dyn_skirt_side_a_base_r}
    rotation {yaw=25}
    translation {y=10, z=-5}
   }
   "Transform (Modify) Bone - Bone: dyn_skirt_bk_a_base_r"!(ComponentPose=@n3, bAlphaBoolEnabled:bool=@n1, Translation="0.005,0,0", Rotation="0,-5,0", Scale="1,1,1") -> (Pose) {
    alphaInputType=Bool
    rotationMode=BMM_Additive
    rotationSpace=BCS_BoneSpace
    translationMode=BMM_Additive
    translationSpace=BCS_BoneSpace
    alphaBoolBlend {blendInTime=0.1, blendOption=Linear, blendOutTime=0.3}
    boneToModify {boneName=dyn_skirt_bk_a_base_r}
    rotation {yaw=-5}
    translation {x=0.005}
   }
   n4 = "Constraint - dyn_skirt_fr_a_base_l (RR)"!(
    ComponentPose=Local To Component(LocalPose=Copy Pose From Mesh() {bUseAttachedParent=true})
    Alpha=1
    ConstraintWeights_0=0
    ConstraintWeights_1:float=1
   ) -> (Pose) {
    alphaBoolBlend {blendOption=Linear}
    boneToModify {boneName=dyn_skirt_fr_a_base_l}
    constraintSetup [{transformType=Rotation, targetBone {boneName=pelvis}}, {transformType=Rotation, targetBone {boneName=thigh_twist_01_l}}]
    constraintWeights [1,1]
   }
   n5 = "Constraint - dyn_skirt_side_a_base_l (RR)"!(ComponentPose=@n4, Alpha=1, ConstraintWeights_0=0, ConstraintWeights_1:float=1) -> (Pose) {
    alphaBoolBlend {blendOption=Linear}
    boneToModify {boneName=dyn_skirt_side_a_base_l}
    constraintSetup [{transformType=Rotation, targetBone {boneName=pelvis}}, {transformType=Rotation, targetBone {boneName=thigh_twist_01_l}}]
    constraintWeights [1,1]
   }
   n6 = "Constraint - dyn_skirt_bk_a_base_l (RR)"!(ComponentPose=@n5, Alpha=1, ConstraintWeights_0=0.9, ConstraintWeights_1:float=0.1) -> (Pose) {
    alphaBoolBlend {blendOption=Linear}
    boneToModify {boneName=dyn_skirt_bk_a_base_l}
    constraintSetup [{transformType=Rotation, targetBone {boneName=pelvis}}, {transformType=Rotation, targetBone {boneName=thigh_l}}]
    constraintWeights [1,1]
   }
   n7 = "Constraint - dyn_skirt_fr_a_base_r (RR)"!(ComponentPose=@n6, Alpha=1, ConstraintWeights_0=0, ConstraintWeights_1:float=1) -> (Pose) {
    alphaBoolBlend {blendOption=Linear}
    boneToModify {boneName=dyn_skirt_fr_a_base_r}
    constraintSetup [{transformType=Rotation, targetBone {boneName=pelvis}}, {transformType=Rotation, targetBone {boneName=thigh_twist_01_r}}]
    constraintWeights [1,1]
   }
   n8 = "Constraint - dyn_skirt_side_a_base_r (RR)"!(ComponentPose=@n7, Alpha=1, ConstraintWeights_0=0, ConstraintWeights_1:float=1) -> (Pose) {
    alphaBoolBlend {blendOption=Linear}
    boneToModify {boneName=dyn_skirt_side_a_base_r}
    constraintSetup [{transformType=Rotation, targetBone {boneName=pelvis}}, {transformType=Rotation, targetBone {boneName=thigh_twist_01_r}}]
    constraintWeights [1,1]
   }
   n9 = "Constraint - dyn_skirt_bk_a_base_r (RR)"!(ComponentPose=@n8, Alpha=1, ConstraintWeights_0=0.9, ConstraintWeights_1:float=0.1) -> (Pose) {
    alphaBoolBlend {blendOption=Linear}
    boneToModify {boneName=dyn_skirt_bk_a_base_r}
    constraintSetup [{transformType=Rotation, targetBone {boneName=pelvis}}, {transformType=Rotation, targetBone {boneName=thigh_r}}]
    constraintWeights [1,1]
   }
   n10 = "Constraint - dyn_skirt_fr_a (PP)"!(ComponentPose=@n9, Alpha=1, ConstraintWeights_0=0.6, ConstraintWeights_1:float=0.4) -> (Pose) {
    alphaBoolBlend {blendOption=Linear}
    boneToModify {boneName=dyn_skirt_fr_a}
    constraintSetup [{targetBone {boneName=dyn_skirt_fr_a_l}}, {targetBone {boneName=dyn_skirt_fr_a_r}}]
    constraintWeights [1,1]
   }
   n11 = "Constraint - dyn_skirt_bk_a_base_r (PPPPPPPP)"!(
    ComponentPose=@n10
    bAlphaBoolEnabled:bool=Get bIsSliding()
    ConstraintWeights_0=0.0164
    ConstraintWeights_1:float=0.01
    ConstraintWeights_2:float=0.0638
    ConstraintWeights_3:float=0.0832
    ConstraintWeights_4:float=0.7365
    ConstraintWeights_5:float=0
    ConstraintWeights_6:float=0.0857
    ConstraintWeights_7:float=0.0043
   ) -> (Pose) {
    alphaInputType=Bool
    alphaBoolBlend {blendInTime=0.1, blendOption=Linear, blendOutTime=0.1}
    boneToModify {boneName=dyn_skirt_bk_a_base_r}
    constraintSetup [
     {targetBone {boneName=spine_04}}
     {targetBone {boneName=spine_03}}
     {targetBone {boneName=spine_02}}
     {targetBone {boneName=spine_01}}
     {targetBone {boneName=pelvis}}
     {targetBone {boneName=thigh_twist_01_l}}
     {targetBone {boneName=thigh_twist_01_r}}
     {targetBone {boneName=thigh_r}}
    ]
    constraintWeights [1,1,1,1,1,1,1,1]
   }
   n12 = "Transform (Modify) Bone - Bone: thigh_twist_01_r"(ComponentPose=@n11, Alpha=1, Translation="0,0,0", Rotation="0,0,0", Scale="0.8,0.8,0.8") {
    scaleMode=BMM_Additive
    scaleSpace=BCS_BoneSpace
    alphaBoolBlend {blendOption=Linear}
    boneToModify {boneName=thigh_twist_01_r}
    scale=<0.8 0.8 0.8>
   }
   n13 = "Transform (Modify) Bone - Bone: dyn_skirt_side_a_base_r"!(ComponentPose=@n12, Alpha:float=1, Translation="-1,2.5,1", Rotation="0,0,0", Scale="1,1,1") -> (Pose) {
    rotationMode=BMM_Additive
    translationMode=BMM_Additive
    translationSpace=BCS_BoneSpace
    alphaBoolBlend {blendOption=Linear}
    boneToModify {boneName=dyn_skirt_side_a_base_r}
    translation=<-1 2.5 1>
   }
   n14 = "Transform (Modify) Bone - Bone: dyn_skirt_fr_a_base_r"!(ComponentPose=@n13, Alpha:float=1, Translation="-1,1,0", Rotation="0,0,0", Scale="1,1,1") -> (Pose) {
    translationMode=BMM_Additive
    translationSpace=BCS_BoneSpace
    alphaBoolBlend {blendOption=Linear}
    boneToModify {boneName=dyn_skirt_fr_a_base_r}
    translation=<-1 1>
   }
   n15 = Get strafe r() -> (strafe r:double)
   n16 = "Transform (Modify) Bone - Bone: dyn_skirt_side_a_base_r"(ComponentPose=@n14, Alpha=@n15, Translation="2,-2.5,-1", Rotation="0,0,0", Scale="1,1,1") {
    rotationMode=BMM_Additive
    translationMode=BMM_Additive
    translationSpace=BCS_BoneSpace
    alphaBoolBlend {blendOption=Linear}
    boneToModify {boneName=dyn_skirt_side_a_base_r}
    translation=<2 -2.5 -1>
   }
   n17 = "Transform (Modify) Bone - Bone: dyn_skirt_fr_a_base_r"(ComponentPose=@n16, Alpha=@n15, Translation="2,-1,0", Rotation="0,0,0", Scale="1,1,1") {
    translationMode=BMM_Additive
    translationSpace=BCS_BoneSpace
    alphaBoolBlend {blendOption=Linear}
    boneToModify {boneName=dyn_skirt_fr_a_base_r}
    translation=<2 -1>
   }
   n18 = Control Rig!(
    Source=Component To Local(ComponentPose=@n17)
    Alpha
    crouch:bool=Get bIsCrouching()
    Moving 2D:bool=Get bIsMoving2D()
    static_rotation:double=Get static_rotation() -> (static_rotation:double)
    Left:double=Get Left() -> (Left:double)
    Right:double=Get Right() -> (Right:double)
    Forward:double=Get Forward() -> (Forward:double)
    Backward:double=Get Backward() -> (Backward:double)
    Targeting:bool=Get bIsTargeting()
    Emote:bool=Get bIsPlayingEmote()
    FrontEnd:bool=Get bIsInFrontEnd()
   ) -> (Pose) {
    alphaBoolBlend {blendOption=Linear}
    controlRigAssetReference {blueprintRigClass=&/Pencil_SkirtShort_MedB_Lbdy_Sailor_CtrlRig.Pencil_SkirtShort_MedB_Lbdy_Sailor_CtrlRig_C}
   }
   n19 = RigidBody(ComponentPose=Local To Component(LocalPose=@n18)) {
    bOverrideWorldGravity=true
    overridePhysicsAsset=&/Pencil_SkirtShort_MedB_Lbdy_Physics
    simulationSpace=BaseBoneSpace
    alphaBoolBlend {blendOption=Linear}
    baseBoneRef {boneName=pelvis}
    componentAppliedLinearAccClamp=<3000 3000 3000>
    componentLinearAccScale=<0.35 0.25 0.25>
    componentLinearVelScale=<0.25 0.25 0.25>
    overrideWorldGravity {z=-980}
    simSpaceSettings {maxAngularAcceleration=500, maxAngularVelocity=3, maxLinearAcceleration=400, maxLinearVelocity=250}
   }
   n20 = RigidBody!(ComponentPose=@n19, bAlphaBoolEnabled:bool=Get bIsFalling(), ExternalForce) -> (Pose) {
    alphaInputType=Bool
    bOverrideWorldGravity=true
    overridePhysicsAsset=&/Pencil_SkirtShort_MedB_Lbdy_Front_Physics1
    simulationSpace=BaseBoneSpace
    alphaBoolBlend {blendInTime=0.5, blendOption=Linear, blendOutTime=0.5}
    baseBoneRef {boneName=pelvis}
    componentAppliedLinearAccClamp=<3000 3000 3000>
    componentLinearAccScale=<0.25 0.25 0.25>
    componentLinearVelScale=<0.25 0.25 0.25>
    overrideWorldGravity {z=-980}
   }
   n21 = RigidBody!(ComponentPose=@n20, Alpha:float, ExternalForce="0,0,-980") -> (Pose) {
    bOverrideWorldGravity=true
    overridePhysicsAsset=&/Pencil_SkirtShort_MedB_Belt_Physics
    simulationSpace=BaseBoneSpace
    alphaBoolBlend {blendOption=Linear}
    baseBoneRef {boneName=pelvis}
    componentAppliedLinearAccClamp=<3000 3000 3000>
    componentLinearAccScale=<0.45 0.25 0.25>
    componentLinearVelScale=<0.25 0.25 0.25>
    overrideWorldGravity {z=-980}
    simSpaceSettings {maxAngularAcceleration=500, maxAngularVelocity=3, maxLinearAcceleration=400, maxLinearVelocity=250}
   }
   n22 = Component To Local(ComponentPose=@n21)
   n23 = "Constraint - dyn_skirt_fr_a_base_l (RR)"(
    ComponentPose=Local To Component(LocalPose=Copy Pose From Mesh() {bUseAttachedParent=true})
    Alpha=1
    ConstraintWeights_0=0
    ConstraintWeights_1=1
   ) {
    alphaBoolBlend {blendOption=Linear}
    boneToModify {boneName=dyn_skirt_fr_a_base_l}
    constraintSetup [{transformType=Rotation, targetBone {boneName=pelvis}}, {transformType=Rotation, targetBone {boneName=thigh_twist_01_l}}]
    constraintWeights [1,1]
   }
   n24 = "Constraint - dyn_skirt_side_a_base_l (RR)"(ComponentPose=@n23, Alpha=1, ConstraintWeights_0=0, ConstraintWeights_1=1) {
    alphaBoolBlend {blendOption=Linear}
    boneToModify {boneName=dyn_skirt_side_a_base_l}
    constraintSetup [{transformType=Rotation, targetBone {boneName=pelvis}}, {transformType=Rotation, targetBone {boneName=thigh_twist_01_l}}]
    constraintWeights [1,1]
   }
   n25 = "Constraint - dyn_skirt_bk_a_base_l (RR)"(ComponentPose=@n24, Alpha=1, ConstraintWeights_0=0.9, ConstraintWeights_1=0.1) {
    alphaBoolBlend {blendOption=Linear}
    boneToModify {boneName=dyn_skirt_bk_a_base_l}
    constraintSetup [{transformType=Rotation, targetBone {boneName=pelvis}}, {transformType=Rotation, targetBone {boneName=thigh_l}}]
    constraintWeights [1,1]
   }
   n26 = "Constraint - dyn_skirt_fr_a_base_r (RR)"(ComponentPose=@n25, Alpha=1, ConstraintWeights_0=0, ConstraintWeights_1=1) {
    alphaBoolBlend {blendOption=Linear}
    boneToModify {boneName=dyn_skirt_fr_a_base_r}
    constraintSetup [{transformType=Rotation, targetBone {boneName=pelvis}}, {transformType=Rotation, targetBone {boneName=thigh_r}}]
    constraintWeights [1,1]
   }
   n27 = "Constraint - dyn_skirt_side_a_base_r (RR)"(ComponentPose=@n26, Alpha=1, ConstraintWeights_0=0.2, ConstraintWeights_1=0.8) {
    alphaBoolBlend {blendOption=Linear}
    boneToModify {boneName=dyn_skirt_side_a_base_r}
    constraintSetup [{transformType=Rotation, targetBone {boneName=pelvis}}, {transformType=Rotation, targetBone {boneName=thigh_r}}]
    constraintWeights [1,1]
   }
   n28 = "Constraint - dyn_skirt_bk_a_base_r (RR)"(ComponentPose=@n27, Alpha=1, ConstraintWeights_0=0.9, ConstraintWeights_1=0.1) {
    alphaBoolBlend {blendOption=Linear}
    boneToModify {boneName=dyn_skirt_bk_a_base_r}
    constraintSetup [{transformType=Rotation, targetBone {boneName=pelvis}}, {transformType=Rotation, targetBone {boneName=thigh_r}}]
    constraintWeights [1,1]
   }
   n29 = "Constraint - dyn_skirt_fr_a (RR)"!(ComponentPose=@n28, Alpha=1, ConstraintWeights_0=0.6, ConstraintWeights_1:float=0.4) -> (Pose) {
    alphaBoolBlend {blendOption=Linear}
    boneToModify {boneName=dyn_skirt_fr_a}
    constraintSetup [{transformType=Rotation, targetBone {boneName=dyn_skirt_fr_a_l}}, {transformType=Rotation, targetBone {boneName=dyn_skirt_fr_a_r}}]
    constraintWeights [1,1]
   }
   n30 = Control Rig!(
    Source=Component To Local(ComponentPose=@n29)
    Alpha
    crouch=Get bIsCrouching()
    Moving 2D=Get bIsMoving2D()
    static_rotation=Get static_rotation()
    Left=Get Left()
    Right=Get Right()
    Forward=Get Forward()
    Backward=Get Backward()
    Targeting=Get bIsTargeting()
    Emote=Get bIsPlayingEmote()
   ) -> (Pose) {
    alphaBoolBlend {blendOption=Linear}
    controlRigAssetReference {blueprintRigClass=&/Pencil_SkirtShort_MedB_Lbdy_CtrlRig.Pencil_SkirtShort_MedB_Lbdy_CtrlRig_C}
   }
   n31 = RigidBody(ComponentPose=Local To Component(LocalPose=@n30)) {
    bOverrideWorldGravity=true
    overridePhysicsAsset=&/Pencil_SkirtShort_Blazer_MedB_Lbdy_Physics
    simulationSpace=BaseBoneSpace
    alphaBoolBlend {blendOption=Linear}
    baseBoneRef {boneName=pelvis}
    componentAppliedLinearAccClamp=<3000 3000 3000>
    componentLinearAccScale=<0.35 0.25 0.25>
    componentLinearVelScale=<0.25 0.25 0.25>
    overrideWorldGravity {z=-980}
    simSpaceSettings {maxAngularAcceleration=1000, maxAngularVelocity=3, maxLinearAcceleration=700, maxLinearVelocity=500, externalLinearDragV=<0.2 0.2>}
   }
   n32 = "Transform (Modify) Bone - Bone: dyn_belt_MSTR"(ComponentPose=@n31, Alpha=1, Translation="0,-0.5,-0.2", Rotation="0,0,0", Scale="1,1,1") {
    translationMode=BMM_Additive
    translationSpace=BCS_BoneSpace
    alphaBoolBlend {blendOption=Linear}
    boneToModify {boneName=dyn_belt_MSTR}
    translation {y=-0.5, z=-0.2}
   }
   n33 = Component To Local(ComponentPose=@n32)
   n34 = Get Jacket OR Blazer() -> (Jacket OR Blazer:bool)
   n35 = "Transform (Modify) Bone - Bone: dyn_belt_MSTR"!(
    ComponentPose=Local To Component(LocalPose=Blend Poses by bool(BlendPose_0=@n33, BlendPose_1=@n22, bActiveValue=@n34) {blendType=Linear})
    bAlphaBoolEnabled:bool=Get Crouching_AND_Moving2D() -> (Crouching_AND_Moving2D:bool)
    Translation="0,-2,0"
    Rotation="0,0,0"
    Scale="1,1,1"
   ) -> (Pose) {
    alphaInputType=Bool
    rotationMode=BMM_Additive
    translationMode=BMM_Additive
    translationSpace=BCS_BoneSpace
    alphaBoolBlend {blendInTime=0.1, blendOption=Linear, blendOutTime=0.1}
    alphaScaleBiasClamp {bInterpResult=true}
    boneToModify {boneName=dyn_belt_MSTR}
    translation {y=-2}
   }
   n36 = "Transform (Modify) Bone - Bone: dyn_skirt_fr_a_base_r"(ComponentPose=@n35, Alpha=Get jump float() -> (jump float:double), Translation="0,2,0", Rotation="0,0,0", Scale="1,1,1") {
    rotationMode=BMM_Additive
    translationMode=BMM_Additive
    translationSpace=BCS_BoneSpace
    alphaBoolBlend {blendOption=Linear}
    alphaScaleBiasClamp {bInterpResult=true}
    boneToModify {boneName=dyn_skirt_fr_a_base_r}
    translation {y=2}
   }
   n37 = Get bIsInFrontEnd()
   n38 = "Transform (Modify) Bone - Bone: dyn_skirt_side_a_base_l"!(ComponentPose=@n36, bAlphaBoolEnabled:bool=@n37, Translation="0,0,0", Rotation="0,5,0", Scale="1,1,1") -> (Pose) {
    alphaInputType=Bool
    rotationMode=BMM_Additive
    rotationSpace=BCS_BoneSpace
    translationMode=BMM_Additive
    translationSpace=BCS_BoneSpace
    alphaBoolBlend {blendOption=Linear}
    boneToModify {boneName=dyn_skirt_side_a_base_l}
    rotation {yaw=5}
   }
   n39 = "Transform (Modify) Bone - Bone: dyn_skirt_side_a_base_r"!(ComponentPose=@n38, bAlphaBoolEnabled:bool=@n37, Translation="0,0,0", Rotation="0,-5,0", Scale="1,1,1") -> (Pose) {
    alphaInputType=Bool
    rotationMode=BMM_Additive
    rotationSpace=BCS_BoneSpace
    translationMode=BMM_Additive
    translationSpace=BCS_BoneSpace
    alphaBoolBlend {blendOption=Linear}
    boneToModify {boneName=dyn_skirt_side_a_base_r}
    rotation {yaw=-5}
   }
   n40 = Get bIsSliding()
   n41 = "Transform (Modify) Bone - Bone: dyn_skirt_side_a_base_l"(ComponentPose=@n39, bAlphaBoolEnabled=@n40, Translation="0,-4,0", Rotation="0,-5,0", Scale="1,1,1") {
    alphaInputType=Bool
    rotationMode=BMM_Additive
    rotationSpace=BCS_BoneSpace
    translationMode=BMM_Additive
    translationSpace=BCS_BoneSpace
    alphaBoolBlend {blendInTime=0.1, blendOption=Linear, blendOutTime=0.3}
    boneToModify {boneName=dyn_skirt_side_a_base_l}
    rotation {yaw=-5}
    translation {y=-4}
   }
   n42 = "Transform (Modify) Bone - Bone: dyn_skirt_side_a_base_r"(ComponentPose=@n41, bAlphaBoolEnabled=@n40, Translation="0,4,0", Rotation="0,10,0", Scale="1,1,1") {
    alphaInputType=Bool
    rotationMode=BMM_Additive
    rotationSpace=BCS_BoneSpace
    translationMode=BMM_Additive
    translationSpace=BCS_BoneSpace
    alphaBoolBlend {blendInTime=0.1, blendOption=Linear, blendOutTime=0.3}
    boneToModify {boneName=dyn_skirt_side_a_base_r}
    rotation {yaw=10}
    translation {y=4}
   }
   n43 = "Transform (Modify) Bone - Bone: dyn_skirt_bk_a_base_l"!(ComponentPose=@n42, bAlphaBoolEnabled:bool=@n40, Translation="5,3,0", Rotation="0,-65,0", Scale="1,1,1") -> (Pose) {
    alphaInputType=Bool
    rotationMode=BMM_Additive
    rotationSpace=BCS_BoneSpace
    translationMode=BMM_Additive
    translationSpace=BCS_BoneSpace
    alphaBoolBlend {blendInTime=0.1, blendOption=Linear, blendOutTime=0.3}
    boneToModify {boneName=dyn_skirt_bk_a_base_l}
    rotation {yaw=-65}
    translation=<5 3>
   }
   n44 = "Transform (Modify) Bone - Bone: dyn_skirt_bk_a_base_r"(ComponentPose=@n43, bAlphaBoolEnabled=@n40, Translation="-5,-5,5", Rotation="15,-80,0", Scale="1,1,1") {
    alphaInputType=Bool
    rotationMode=BMM_Additive
    rotationSpace=BCS_BoneSpace
    translationMode=BMM_Additive
    translationSpace=BCS_BoneSpace
    alphaBoolBlend {blendInTime=0.1, blendOption=Linear, blendOutTime=0.3}
    boneToModify {boneName=dyn_skirt_bk_a_base_r}
    rotation {pitch=15, yaw=-80}
    translation=<-5 -5 5>
   }
   n45 = "Transform (Modify) Bone - Bone: dyn_skirt_bk_a_base"!(ComponentPose=@n44, bAlphaBoolEnabled:bool=@n40, Translation="-5,-3,0", Rotation="0,-80,0", Scale="1,1,1") -> (Pose) {
    alphaInputType=Bool
    rotationMode=BMM_Additive
    rotationSpace=BCS_BoneSpace
    translationMode=BMM_Additive
    translationSpace=BCS_BoneSpace
    alphaBoolBlend {blendInTime=0.1, blendOption=Linear, blendOutTime=0.3}
    boneToModify {boneName=dyn_skirt_bk_a_base}
    rotation {yaw=-80}
    translation=<-5 -3>
   }
   n46 = Get bIsRiding()
   n47 = "Transform (Modify) Bone - Bone: dyn_skirt_fr_a_base_r"!(ComponentPose=@n45, bAlphaBoolEnabled:bool=@n46, Translation="0.005,3,-5", Rotation="0,5,0", Scale="1,1,1") -> (Pose) {
    alphaInputType=Bool
    rotationMode=BMM_Additive
    rotationSpace=BCS_BoneSpace
    translationMode=BMM_Additive
    translationSpace=BCS_BoneSpace
    alphaBoolBlend {blendInTime=0.1, blendOption=Linear, blendOutTime=0.3}
    boneToModify {boneName=dyn_skirt_fr_a_base_r}
    rotation {yaw=5}
    translation=<0.005 3 -5>
   }
   n48 = "Transform (Modify) Bone - Bone: dyn_skirt_side_a_base_r"(ComponentPose=@n47, bAlphaBoolEnabled=@n46, Translation="-3,6,-3", Rotation="0,5,0", Scale="1,1,1") {
    alphaInputType=Bool
    rotationMode=BMM_Additive
    rotationSpace=BCS_BoneSpace
    translationMode=BMM_Additive
    translationSpace=BCS_BoneSpace
    alphaBoolBlend {blendInTime=0.1, blendOption=Linear, blendOutTime=0.3}
    boneToModify {boneName=dyn_skirt_side_a_base_r}
    rotation {yaw=5}
    translation=<-3 6 -3>
   }
   n49 = "Transform (Modify) Bone - Bone: dyn_skirt_bk_a_base_r"(ComponentPose=@n48, bAlphaBoolEnabled=@n46, Translation="0.005,-3,-2", Rotation="-15,-5,0", Scale="1,1,1") {
    alphaInputType=Bool
    rotationMode=BMM_Additive
    rotationSpace=BCS_BoneSpace
    translationMode=BMM_Additive
    translationSpace=BCS_BoneSpace
    alphaBoolBlend {blendInTime=0.1, blendOption=Linear, blendOutTime=0.3}
    boneToModify {boneName=dyn_skirt_bk_a_base_r}
    rotation {pitch=-15, yaw=-5}
    translation=<0.005 -3 -2>
   }
   n50 = "Transform (Modify) Bone - Bone: dyn_skirt_fr_a_base_l"!(ComponentPose=@n49, bAlphaBoolEnabled:bool=@n46, Translation="0.005,-4,3", Rotation="0,5,0", Scale="1,1,1") -> (Pose) {
    alphaInputType=Bool
    rotationMode=BMM_Additive
    rotationSpace=BCS_BoneSpace
    translationMode=BMM_Additive
    translationSpace=BCS_BoneSpace
    alphaBoolBlend {blendInTime=0.1, blendOption=Linear, blendOutTime=0.3}
    boneToModify {boneName=dyn_skirt_fr_a_base_l}
    rotation {yaw=5}
    translation=<0.005 -4 3>
   }
   n51 = "Transform (Modify) Bone - Bone: dyn_skirt_side_a_base_l"(ComponentPose=@n50, bAlphaBoolEnabled=@n46, Translation="-3,-4,0", Rotation="0,5,0", Scale="1,1,1") {
    alphaInputType=Bool
    rotationMode=BMM_Additive
    rotationSpace=BCS_BoneSpace
    translationMode=BMM_Additive
    translationSpace=BCS_BoneSpace
    alphaBoolBlend {blendInTime=0.1, blendOption=Linear, blendOutTime=0.3}
    boneToModify {boneName=dyn_skirt_side_a_base_l}
    rotation {yaw=5}
    translation=<-3 -4>
   }
   n52 = Control Rig!(Source=Component To Local(ComponentPose=@n51), Alpha, emote:double=Get emote() -> (emote:double)) -> (Pose) {
    alphaBoolBlend {blendOption=Linear}
    controlRigAssetReference {blueprintRigClass=&/Pencil_SkirtShort_MedB_Lbdy_CtrlRig1.Pencil_SkirtShort_MedB_Lbdy_CtrlRig1_C}
   }
   Output Pose(Result=@n52) {layerGroup=None}
   Get bIsInVehicle() -> (bIsInVehicle:bool)
  ]}
 EventGraph {nodes [
   Event BlueprintUpdateAnimation()
   > n1 = Sequence!(execute) -> (then_0, then_1, then_2:Exec, then_3:Exec, then_4:Exec, then_5:Exec, then_6:Exec, then_7:Exec)
   n2 = BreakRotator(InRot=Get PawnDeltaRotation() -> (PawnDeltaRotation:Rotator))
   > n3 = Branch(Condition="float > float"(A="Absolute (Float)"(A=@n2.Yaw)))
   > Set static_rotation(static_rotation:double=1) -> (Output_Get:double)
   Set static_rotation(execute=@n3.else, static_rotation=0)
   Set Jumping_OR_Parachute(execute=@n1.then_1, Jumping_OR_Parachute:bool=OR Boolean(A=Get bIsJumping(), B=Get bIsParachuteOpened())) -> (Output_Get:bool)
   > Set Parachute_OR_Crouching(Parachute_OR_Crouching:bool=OR Boolean(A=Get bIsParachuteOpened(), B=Get bIsCrouching())) -> (Output_Get:bool)
   > Set Crouching_AND_Moving2D(Crouching_AND_Moving2D:bool=AND Boolean(A=Get bIsCrouching(), B=Get bIsMoving2D())) -> (Output_Get:bool)
   > Set notIdlePose(notIdlePose:bool=OR Boolean!(A=Get bIsMoving2D(), B=Get bIsJumping(), C:bool=Get bIsCrouching(), D:bool=Get bIsParachuteOpened(), E:bool=Get bIsPlayingEmote(), F:bool) -> (ReturnValue)) -> (Output_Get:bool)
   > Set Jacket OR Blazer(Jacket OR Blazer:bool=OR Boolean!(A=Get VestJacket() -> (VestJacket:bool), B=Get BlazerCuff() -> (BlazerCuff:bool)) -> (ReturnValue)) -> (Output_Get:bool)
   n4 = PlayerMovement(execute=@n1.then_2, AnimInstance=Self-Reference(), XDotProductCoefficient=1, YDotProductCoefficient=1, ZDotProductCoefficient=1)
   n5 = MapRangeClamped(Value=@n4, InRangeA=0, InRangeB=1)
   > Set strafe r(strafe r:double=@n5) -> (Output_Get:double)
   n6 = MapRangeClamped(Value=@n4, InRangeA=0, InRangeB=-1)
   > Set strafe l(strafe l:double=@n6) -> (Output_Get:double)
   > n7 = Branch(Condition=Get bIsCrouching())
   > Set crouch strafe r(crouch strafe r:double=@n5) -> (Output_Get:double)
   > Set crouch strafe l(crouch strafe l:double=@n6) -> (Output_Get:double)
   Set crouch strafe r(execute=@n7.else, crouch strafe r=0)
   > Set crouch strafe l(crouch strafe l=0)
   n8 = Get bIsMoving2D()
   n9 = Branch(execute=@n1.then_3, Condition=@n8)
   > Set Moving 2D(Moving 2D:double=1) -> (Output_Get:double)
   Set Moving 2D(execute=@n9.else, Moving 2D=0)
   n10 = PlayerMovement(execute=@n1.then_4, AnimInstance=Self-Reference(), XDotProductCoefficient=100, YDotProductCoefficient=100, ZDotProductCoefficient=100)
   n11 = BreakVector(InVec=MakeVector(X=@n10, Y=@n10.YMovement))
   n12 = "float < float"(A=@n11, B=-50)
   > Set Left(Left:double=ClampAxis!(Angle:float="To Float (Boolean)"(InBool=@n12)) -> (ReturnValue:float)) -> (Output_Get:double)
   n13 = "float > float"(A=@n11, B=50)
   > Set Right(Right:double=ClampAxis(Angle="To Float (Boolean)"(InBool=@n13))) -> (Output_Get:double)
   n14 = "float > float"(A=@n11.Y, B=40)
   > Set Forward(Forward:double=ClampAxis(Angle="To Float (Boolean)"(InBool=@n14))) -> (Output_Get:double)
   n15 = "float < float"(A=@n11.Y, B=-40)
   > Set Backward(Backward:double=ClampAxis(Angle="To Float (Boolean)"(InBool=@n15))) -> (Output_Get:double)
   n16 = Branch(execute=@n1.then_6, Condition=Get bIsPlayingEmote())
   > Set emote(emote:double=1) -> (Output_Get:double)
   Set emote(execute=@n16.else, emote=0)
   Set VestJacket(
    execute=@n1.then_5
    VestJacket:bool=CharacterItemDefinitionHasAnyMetaTag(Tags=Make Array("[0]"=(TagName="Cosmetics.MetaTags.Pencil.Upper.VestJacket")))
   ) -> (Output_Get:bool)
   > Set BlazerCuff(BlazerCuff:bool=CharacterItemDefinitionHasAnyMetaTag(Tags=Make Array("[0]"=(TagName="Cosmetics.MetaTags.Pencil.Upper.BlazerCuff")))) -> (Output_Get:bool)
   n17 = Event OnBeginJumping()
   Set jump float(execute=[@n17.then, @n18.then], jump float:double=1) -> (Output_Get:double)
   n18 = Event OnBeginFalling()
   n19 = Event OnEndJumping()
   Set jump float(execute=[@n19.then, @n20.then], jump float=0)
   n20 = Event OnEndFalling()
   Event OnEndCrouching()
   > Set crouch(crouch:double=0) -> (Output_Get:double)
   Event OnBeginCrouching()
   > Set crouch(crouch=1)
   AND Boolean(A=@n8, B=NOT Boolean(A=Get bIsCrouching()))
   AND Boolean!(A=false, B=NOT Boolean(A=Get bIsCrouching()), C:bool, D:bool) -> (ReturnValue)
  ]}
]
variables [
 Backward {type=number}
 BlazerCuff {type=boolean}
 crouch {type=number}
 crouch strafe l {type=number}
 crouch strafe r {type=number}
 Crouching AND Moving 2D {type=boolean}
 Crouching_AND_Moving2D {type=boolean}
 dyn_skirt_fr_a_base_r_Var {type=Transform, default {location=<0 0 0>, rotation=<rot 0 0 0>, scale=<1 1 1>}}
 dyn_skirt_side_a_base_r_var {type=Transform, default {location=<0 0 0>, rotation=<rot 0 0 0>, scale=<1 1 1>}}
 emote {type=number}
 Forward {type=number}
 Jacket OR Blazer {type=boolean}
 jump float {type=number}
 Jumping_OR_Parachute {type=boolean}
 L_swinging_back_skirt {type=number}
 Left {type=number}
 Left_move {type=boolean}
 Moving 2D {type=number}
 notIdlePose {type=boolean}
 Parachute_OR_Crouching {type=boolean}
 R_swinging_back_skirt {type=number}
 Right {type=number}
 Right_move {type=boolean}
 Skirt Fr A Base L Var {type=number}
 Skirt Fr A Base R Var {type=number}
 skirt_fr_a_base_l_var {type=number}
 skirt_fr_a_base_r_var {type=number}
 Static Rotation {type=number}
 static_rotation {type=number}
 strafe l {type=number}
 strafe r {type=number}
 VestJacket {type=boolean}
]
